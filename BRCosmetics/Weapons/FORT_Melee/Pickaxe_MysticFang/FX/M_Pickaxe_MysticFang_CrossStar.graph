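# Translucent, unlit, two-sided Niagara sprite material that draws a cross-star effect for the Mystic Fang pickaxe.
# The ProceduralCrossStar function generates the star shape, tuned by Tighten and Softness parameters.
# Particle alpha multiplies the shape, and a parameterized DepthFade softens its intersections with scene geometry.
# Particle RGB supplies the color through the shared MF_FX_EmissiveAdjustments function.
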Material /BRCosmetics/Weapons/FORT_Melee/Pickaxe_MysticFang/FX/M_Pickaxe_MysticFang_CrossStar
graph {
 properties {
  bAutomaticallySetUsageInEditor=false
  bUsedWithNiagaraSprites=true
  blendMode=BLEND_Translucent
  shadingModel=MSM_Unlit
  twoSided=true
  usedShadingModels=Unlit
 }
 nodes [
  n1 = ParticleColor()
  n2 = ScalarParameter() {defaultValue=2, parameterName=Tighten, sliderMax=5}
  n3 = ScalarParameter() {parameterName=Softness, sliderMax=5, sliderMin=-5}
  DepthFade(
   Opacity=Multiply(
    A=@n1.A
    B=MaterialFunctionCall!(Tighten=@n2, UVIn, "Softness(0-1)"=@n3, "Fix_AntiFlick(Increase This If The Tip Flicking)", Fix_TipWidth) {materialFunction=/Game/Effects/Niagara/MasterMaterials/Functions/ProceduralCrossStar}.Result
   )
   FadeDistance=ScalarParameter() {defaultValue=30, parameterName=FadeDistance, sliderMax=100}
  )
  MaterialFunctionCall(Emissive=@n1.RGB) {materialFunction=&MF_FX_EmissiveAdjustments}
 ]
}
