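# Translucent unlit Niagara mesh-particle material for the golden spline effects on the Mystic Fang pickaxe.
# SplineThicken sets strand width, with tip width scaled by dynamic Width and a sine taper along V.
# Dynamic Fade Start and Length position a sine-shaped opacity band along V, offset by vertex red.
# A powered sine across U softens edges; particle color feeds emissive adjustments and its alpha scales opacity.
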Material /BRCosmetics/Weapons/FORT_Melee/Pickaxe_MysticFang/FX/M_Pickaxe_MysticFang_GoldenSplines
graph {
 properties {
  bAutomaticallySetUsageInEditor=false
  bUsedWithNiagaraMeshParticles=true
  blendMode=BLEND_Translucent
  shadingModel=MSM_Unlit
  usedShadingModels=Unlit
 }
 nodes [
  n1 = ScalarParameter() {defaultValue=5, parameterName=Spline Width Tip, sliderMax=20}
  n2 = DynamicParameter() {defaultValue=<rgba 0 0.5 1 1>, paramNames [Fade Start, Fade Length, Width, Param4]}
  n3 = ComponentMask(None=TextureCoordinate()) {g=true}
  MaterialFunctionCall(
   WidthBase=ScalarParameter() {defaultValue=1, parameterName=Spline Width Base, sliderMax=10}
   WidthTip=Multiply(A=Multiply(A=@n1, B=@n2.Width), B=Sine(None=@n3) {period=2})
  ) {materialFunction=&SplineThicken}
  n4 = ParticleColor()
  MaterialFunctionCall(Emissive=@n4.RGB) {materialFunction=&MF_FX_EmissiveAdjustments}
  n5 = Add(
   A=ComponentMask(None=TextureCoordinate()) {g=true}
   B=Multiply(A=VertexColor().R, B=ScalarParameter() {parameterName=Vertex Color Offset Intensity, sliderMax=2})
  )
  n6 = DynamicParameter() {defaultValue=<rgba 0 0.5 1 1>, paramNames [Fade Start, Fade Length, Width, Param4]}
  n7 = Saturate(None=Multiply(A=Add(A=@n5, B=@n6.Fade Start), B=@n6.Fade Length) {constB=3})
  n8 = Sine(None=ComponentMask(None=TextureCoordinate()) {r=true}) {period=2}
  Multiply(A=@n4.A, B=Multiply(A=Sine(None=@n7) {period=2}, B=Power(Base=@n8) {constExponent=5}))
 ]
}
