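# Translucent, unlit, two-sided Niagara sprite material that draws a flare shape for the Mystic Fang pickaxe effects.
# It thresholds a grayscale distance-field texture by Shape Size and Shape Softness, each scaled by a particle dynamic parameter.
# The resulting mask, multiplied by particle alpha, feeds a depth fade for soft intersections.
# The same mask, saturated and shaped by a power, blends outer to inner color before a shared emissive adjustment function.
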Material /BRCosmetics/Weapons/FORT_Melee/Pickaxe_MysticFang/FX/M_Sprite_SDF_Flare
graph {
 properties {
  bAutomaticallySetUsageInEditor=false
  bUsedWithNiagaraSprites=true
  blendMode=BLEND_Translucent
  shadingModel=MSM_Unlit
  twoSided=true
  usedShadingModels=Unlit
 }
 nodes [
  n1 = TextureSampleParameter2D() {
   parameterName=Sprite Texture
   samplerType=SAMPLERTYPE_LinearGrayscale
   texture=&/T_Pickaxe_MysticFang_Spikes
  }
  n2 = Multiply(
   A=ScalarParameter() {defaultValue=1, group=Shape, parameterName=Shape Size, sliderMax=1}
   B=DynamicParameter() {paramNames [Shape Size, Shape Softness, Param3, Param4]}.Shape Size
  )
  n3 = ScalarParameter() {defaultValue=1, group=Shape, parameterName=Shape Softness, sliderMax=1}
  n4 = DynamicParameter() {paramNames [Shape Size, Shape Softness, Param3, Param4]}
  n5 = Divide(A=Subtract(A=@n1.R, B=OneMinus(None=@n2)), B=Add(A=Multiply(A=@n3, B=@n4.Shape Softness)) {constB=0.0001})
  DepthFade(Opacity=Multiply(A=@n5, B=ParticleColor().A), FadeDistance=ScalarParameter() {defaultValue=30, parameterName=FadeDistance, sliderMax=100})
  n6 = Saturate(None=@n5)
  n7 = ScalarParameter() {defaultValue=1, parameterName=Color Distribute, sliderMax=5}
  MaterialFunctionCall(Emissive=LinearInterpolate(
    A=VectorParameter() {parameterName=Outer Color}.RGB
    B=VectorParameter() {parameterName=Inner Color, defaultValue=<rgba 1.5 0.8516 0.9264 1>}.RGB
    Alpha=Power(Base=@n6, Exp=@n7)
   )) {materialFunction=&MF_FX_EmissiveAdjustments}
 ]
}
