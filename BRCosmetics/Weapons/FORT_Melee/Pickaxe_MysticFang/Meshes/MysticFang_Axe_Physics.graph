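# Physics asset for the Mystic Fang pickaxe mesh, giving collision and joints to two dangling strip chains on the weapon.
# A multi-capsule dyn_broom_01 body anchors both chains: strip A through a fully swing- and twist-locked sphere base, strip B directly.
# Strip links are mostly single capsules with wide swing and tight twist, except strip A's free-swinging root and looser-twisting tip.
# Both chain tips use compound capsule shapes.
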PhysicsAsset /BRCosmetics/Weapons/FORT_Melee/Pickaxe_MysticFang/Meshes/MysticFang_Axe_Physics
bodies {
 base_stripA_01 [{shapeType=Sphere, radius=3.5, shapeName=base_stripA_01_sphere, center=<-2.6159 0.9049 2.2879>}]
 dyn_broom_01 [
  {shapeType=Capsule, length=29, radius=7, shapeName=dyn_broom_01_capsule, center=<21.6294 -0.2473 -0.1576>, rotation=<rot 0 88 -88>}
  {shapeType=Capsule, length=8, radius=4.5, shapeName=dyn_broom_01_capsule, center=<-3.4606 0.1184 -0.9338>, rotation=<rot 18.3468 0 0>}
  {shapeType=Capsule, length=9.2943, radius=5.2057, shapeName=dyn_broom_01_capsule, center=<-2.5252 0.1571 0.9575>, rotation=<rot 51.4491 0 0>}
  {shapeType=Capsule, length=23.5, radius=6.5, shapeName=dyn_broom_01_capsule, center=<24.2403 0 4.9574>, rotation=<rot -71.9724 0 0>}
  {shapeType=Capsule, length=23.5, radius=6.5, shapeName=dyn_broom_01_capsule, center=<23.044 0 -4.8516>, rotation=<rot -73.9613 180 -180>}
 ]
 dyn_stripA_01 [{shapeType=Capsule, length=4, radius=0.8, shapeName=dyn_stripA_01_capsule, center=<4 0 0>, rotation=<rot 0 -90 -90>}]
 dyn_stripA_02 [{shapeType=Capsule, length=5.8616, radius=0.8, shapeName=dyn_stripA_02_capsule, center=<5.047 0 0>, rotation=<rot 0 -90 -90>}]
 dyn_stripA_03 [
  {shapeType=Capsule, length=5.8, radius=4.7, shapeName=dyn_stripA_03_capsule, center=<5.2331 -0.1054 -1.4943>, rotation=<rot 0 19.5 90>}
  {shapeType=Capsule, length=5.9947, radius=0.7421, shapeName=dyn_stripA_03_capsule, center=<7.4086 2.8685 0.2976>, rotation=<rot 0 -120 -90>}
  {shapeType=Capsule, length=3, radius=1, shapeName=dyn_stripA_03_capsule, center=<3.041 2.453 0.2976>, rotation=<rot 0 -40 -90>}
  {
   shapeType=Capsule
   length=5.9715
   radius=0.8434
   shapeName=dyn_stripA_03_capsule
   center=<7.4612 -1.5765 0.2976>
   rotation=<rot 0 -63.8457 -90>
  }
  {shapeType=Capsule, length=2.8572, radius=0.9678, shapeName=dyn_stripA_03_capsule, center=<3.15 -1.15 0.3>, rotation=<rot 0 -135 -90>}
 ]
 dyn_stripB_01 [{shapeType=Capsule, length=6.0818, radius=0.8, shapeName=dyn_stripB_01_capsule, center=<10 0 0>, rotation=<rot 0 90 -89>}]
 dyn_stripB_02 [{shapeType=Capsule, length=8.294, radius=1, shapeName=dyn_stripB_02_capsule, center=<10.8366 0 0.1404>, rotation=<rot 0 -90 -90>}]
 dyn_stripB_03 [{shapeType=Capsule, length=5.4477, radius=0.9, shapeName=dyn_stripB_03_capsule, center=<10 0 0>, rotation=<rot 0 -90 -90>}]
 dyn_stripB_04 [
  {shapeType=Capsule, length=13, radius=0.9, shapeName=dyn_stripB_04_capsule, center=<7.7502 0.0148 0.1328>, rotation=<rot 0 90 -88>}
  {
   shapeType=Capsule
   length=7.75
   radius=0.9
   shapeName=dyn_stripB_04_capsule
   center=<9.4238 0.0435 -0.158>
   rotation=<rot -2 179.802 -89.9931>
  }
 ]
}
constraints [
 {bone1Name=dyn_stripA_02, bone2Name=dyn_stripA_01, swing1LimitDegrees=80, swing2LimitDegrees=80, twistLimitDegrees=5}
 {bone1Name=dyn_stripA_03, bone2Name=dyn_stripA_02, swing1LimitDegrees=80, swing2LimitDegrees=80, twistLimitDegrees=30}
 {bone1Name=dyn_stripB_04, bone2Name=dyn_stripB_03, swing1LimitDegrees=80, swing2LimitDegrees=80, twistLimitDegrees=2}
 {bone1Name=dyn_stripB_02, bone2Name=dyn_stripB_01, swing1LimitDegrees=80, swing2LimitDegrees=80, twistLimitDegrees=2}
 {bone1Name=dyn_stripB_03, bone2Name=dyn_stripB_02, swing1LimitDegrees=80, swing2LimitDegrees=80, twistLimitDegrees=2}
 {
  bone1Name=base_stripA_01
  bone2Name=dyn_broom_01
  swing1LimitDegrees=45
  swing1Motion=Locked
  swing2LimitDegrees=45
  swing2Motion=Locked
  twistLimitDegrees=45
  twistMotion=Locked
 }
 {bone1Name=dyn_stripB_01, bone2Name=dyn_broom_01, swing1LimitDegrees=80, swing2LimitDegrees=80, twistLimitDegrees=2}
 {
  bone1Name=dyn_stripA_01
  bone2Name=base_stripA_01
  swing1LimitDegrees=80
  swing1Motion=Free
  swing2LimitDegrees=80
  swing2Motion=Free
  twistLimitDegrees=2
 }
]
