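# Anim Blueprint for the Mystic Fang pickaxe, adding secondary motion to its dangling strips and bristles.
# With rigid bodies enabled, Anim Dynamics drives the bristles before a RigidBody simulation and corrective bone offsets.
# Otherwise, Trail controllers run on a pre-posed cache, with separate front-end and in-game chains.
# MovingAlpha weights the in-game trails, but the jump/fall check overwrites the sprint/move value, leaving 0.5 airborne, otherwise 1.
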AnimBlueprint /BRCosmetics/Weapons/FORT_Melee/Pickaxe_MysticFang/Meshes/MysticFang_Axe_AnimBP
target_skeleton=&/MysticFang_Axe_Skeleton
graphs [
 AnimGraph {nodes [
   n1 = "Transform (Modify) Bone - Bone: base_stripA_01"(ComponentPose=Local To Component(LocalPose=Use cached pose 'TrailPose'()), Alpha=1, Translation="0,0,0", Rotation="90,120,0", Scale="1,1,1") {rotationMode=BMM_Additive, alphaBoolBlend {blendOption=Linear}, boneToModify {boneName=base_stripA_01}, rotation {pitch=90, yaw=120}}
   n2 = "Transform (Modify) Bone - Bone: base_stripB_01"(ComponentPose=@n1, Alpha=1, Translation="0,0,0", Rotation="110,120,0", Scale="1,1,1") {rotationMode=BMM_Additive, alphaBoolBlend {blendOption=Linear}, boneToModify {boneName=base_stripB_01}, rotation {pitch=110, yaw=120}}
   n3 = "Transform (Modify) Bone - Bone: dyn_stripB_02"(ComponentPose=@n2, Alpha=1, Translation="0,0,0", Rotation="0,0,20", Scale="1,1,1") {rotationMode=BMM_Additive, alphaBoolBlend {blendOption=Linear}, boneToModify {boneName=dyn_stripB_02}, rotation {roll=20}}
   n4 = Trail controller - Bone: dyn_stripA_03(ComponentPose=@n3, Alpha=1) {
    bLimitRotation=true
    chainLength=3
    maxDeltaTime=0.333
    relaxationSpeedScale=2
    alphaScaleBiasClamp {bInterpResult=true}
    rotationLimits [{limitMax=<180 180 180>}, {limitMax=<180 180 180>}, {limitMax=<180 180 180>}]
    rotationOffsets [<0 0 0>, <0 0 0>, <0 0 0>]
    trailBone {boneName=dyn_stripA_03}
   }
   n5 = Trail controller - Bone: dyn_stripB_04(ComponentPose=@n4, Alpha=1) {
    bLimitRotation=true
    chainLength=4
    maxDeltaTime=0.333
    relaxationSpeedScale=1.2
    alphaScaleBiasClamp {bInterpResult=true}
    rotationLimits [{limitMax=<180 180 180>}, {limitMax=<180 180 180>}, {limitMax=<180 180 180>}, {limitMax=<180 180 180>}]
    rotationOffsets [<0 0 0>, <0 0 0>, <0 0 0>, <0 0 0>]
    trailBone {boneName=dyn_stripB_04}
   }
   n6 = Component To Local(ComponentPose=@n5)
   n7 = Get MovingAlpha() -> (MovingAlpha:double)
   n8 = Trail controller - Bone: dyn_stripA_03(ComponentPose=Local To Component(LocalPose=Use cached pose 'TrailPose'()), Alpha=@n7) {
    bLimitRotation=true
    chainLength=3
    maxDeltaTime=0.333
    relaxationSpeedScale=2.5
    alphaScaleBiasClamp {bInterpResult=true}
    rotationLimits [{limitMax=<30 30 30>, limitMin=<-10 -30 -30>}, {limitMax=<60 60 60>, limitMin=<-30 -60 -60>}, {limitMax=<50 50 50>, limitMin=<-50 -50 -50>}]
    rotationOffsets [<0 0 0>, <0 0 0>, <0 0 0>]
    trailBone {boneName=dyn_stripA_03}
   }
   n9 = Trail controller - Bone: dyn_stripB_04(ComponentPose=@n8, Alpha=@n7) {
    bLimitRotation=true
    chainLength=3
    maxDeltaTime=0.333
    relaxationSpeedScale=1.5
    alphaScaleBiasClamp {bInterpResult=true}
    relaxationSpeedScaleInputProcessor {interpSpeedDecreasing=50, interpSpeedIncreasing=50}
    rotationLimits [{limitMax=<20 20 20>, limitMin=<-20 -20 -20>}, {limitMax=<40 40 40>, limitMin=<-20 -20 -60>}, {limitMax=<60 60 60>, limitMin=<-20 -60 -60>}]
    rotationOffsets [<0 0 0>, <0 0 0>, <0 0 0>]
    trailBone {boneName=dyn_stripB_04}
    trailRelaxationSpeed {editorCurveData {keys [<at 0 10 RCIM_Linear>,<at 1 6 RCIM_Linear>]}}
   }
   n10 = Component To Local(ComponentPose=@n9)
   n11 = Get bIsInFrontEnd()
   Output Pose(Result=Blend Poses by bool(
     BlendPose_0=Use cached pose 'RBANPose'()
     BlendPose_1=Blend Poses by bool(BlendPose_0=@n6, BlendPose_1=@n10, bActiveValue=@n11) {blendType=Linear}
     bActiveValue=Get bRigidBodyEnabled()
    ) {blendType=Linear}) {layerGroup=None}
   AnimGraphNode_SubInput_0 = Input Pose()
   Save cached pose 'PrePose'(Pose=@AnimGraphNode_SubInput_0) {cacheName=PrePose}
   n12 = "Transform (Modify) Bone - Bone: dyn_stripA_01"(ComponentPose=Local To Component(LocalPose=Use cached pose 'PrePose'()), Alpha=1, Translation="0,0,0", Rotation="-5,-10,0", Scale="1,1,1") {
    rotationMode=BMM_Additive
    rotationSpace=BCS_BoneSpace
    translationMode=BMM_Additive
    translationSpace=BCS_BoneSpace
    alphaBoolBlend {blendOption=Linear}
    alphaScaleBiasClamp {bInterpResult=true}
    boneToModify {boneName=dyn_stripA_01}
    rotation {pitch=-5, yaw=-10}
   }
   n13 = "Transform (Modify) Bone - Bone: dyn_stripA_02"(ComponentPose=@n12, Alpha=1, Translation="0,0,0", Rotation="-15,0,0", Scale="1,1,1") {
    rotationMode=BMM_Additive
    rotationSpace=BCS_BoneSpace
    translationMode=BMM_Additive
    translationSpace=BCS_BoneSpace
    alphaBoolBlend {blendOption=Linear}
    alphaScaleBiasClamp {bInterpResult=true}
    boneToModify {boneName=dyn_stripA_02}
    rotation {pitch=-15}
   }
   n14 = "Transform (Modify) Bone - Bone: base_stripA_01"(ComponentPose=@n13, Alpha=1, Translation="0,0,0", Rotation="-105,0,0", Scale="1,1,1") {
    rotationMode=BMM_Additive
    rotationSpace=BCS_BoneSpace
    translationMode=BMM_Additive
    translationSpace=BCS_BoneSpace
    alphaBoolBlend {blendOption=Linear}
    alphaScaleBiasClamp {bInterpResult=true}
    boneToModify {boneName=base_stripA_01}
    rotation {pitch=-105}
   }
   n15 = "Transform (Modify) Bone - Bone: dyn_stripB_01"(ComponentPose=@n14, Alpha=1, Translation="1,0,3", Rotation="-8,-20,0", Scale="1,1,1") {
    rotationMode=BMM_Additive
    rotationSpace=BCS_BoneSpace
    translationMode=BMM_Additive
    translationSpace=BCS_BoneSpace
    alphaBoolBlend {blendOption=Linear}
    alphaScaleBiasClamp {bInterpResult=true}
    boneToModify {boneName=dyn_stripB_01}
    rotation {pitch=-8, yaw=-20}
    translation {x=1, z=3}
   }
   n16 = "Transform (Modify) Bone - Bone: dyn_stripB_02"(ComponentPose=@n15, Alpha=1, Translation="0,0,0", Rotation="-10,-10,0", Scale="1,1,1") {
    rotationMode=BMM_Additive
    rotationSpace=BCS_BoneSpace
    translationMode=BMM_Additive
    translationSpace=BCS_BoneSpace
    alphaBoolBlend {blendOption=Linear}
    alphaScaleBiasClamp {bInterpResult=true}
    boneToModify {boneName=dyn_stripB_02}
    rotation {pitch=-10, yaw=-10}
   }
   n17 = "Transform (Modify) Bone - Bone: base_stripB_01"(ComponentPose=@n16, Alpha=1, Translation="0,0,0", Rotation="-85,0,0", Scale="1,1,1") {
    rotationMode=BMM_Additive
    rotationSpace=BCS_BoneSpace
    translationMode=BMM_Additive
    translationSpace=BCS_BoneSpace
    alphaBoolBlend {blendOption=Linear}
    alphaScaleBiasClamp {bInterpResult=true}
    boneToModify {boneName=base_stripB_01}
    rotation {pitch=-85}
   }
   Save cached pose 'TrailPose'(Pose=Component To Local(ComponentPose=@n17)) {cacheName=TrailPose}
   n18 = Anim Dynamics - Bone: dyn_bristles_01(ComponentPose=Local To Component(LocalPose=Use cached pose 'PrePose'()), Alpha=1, ExternalForce="0,0,0") {
    angularDampingOverride=0.9999
    angularSpringConstant=10000
    bAngularSpring=true
    bGravityOverrideInSimSpace=true
    bOverrideAngularDamping=true
    bUseGravityOverride=true
    gravityScale=0
    simulationSpace=RootRelative
    alphaBoolBlend {blendOption=Linear}
    alphaScaleBiasClamp {bClampResult=true}
    boundBone {boneName=dyn_bristles_01}
    componentAppliedLinearAccClamp {y=10000, z=10000}
    componentLinearAccScale=<0.5 0.5 0.5>
    componentLinearVelScale=<0.5 0.5 0.5>
    gravityOverride {z=980}
    physicsBodyDefinitions [{
      boundBone {boneName=dyn_bristles_01}
      boxExtents=<50 50 50>
      constraintSetup {angularLimitsMax=<0 5 5>, angularLimitsMin=<0 -5 -5>, angularTarget=<1 0 0>}
      localJointOffset=<4 0 0>
     }]
   }
   n19 = RigidBody(ComponentPose=@n18) {
    bOverrideWorldGravity=true
    overridePhysicsAsset=&/MysticFang_Axe_Physics
    simulationSpace=BaseBoneSpace
    alphaBoolBlend {blendOption=Linear}
    baseBoneRef {boneName=b_AR_Root}
    componentAppliedLinearAccClamp=<2000 2000 2000>
    overrideWorldGravity {z=-980}
    simSpaceSettings {
     dampingAlpha=0.7
     maxAngularAcceleration=50
     maxAngularVelocity=10
     maxLinearAcceleration=3000
     maxLinearVelocity=600
     velocityScaleZ=0.7
     worldAlpha=0.8
    }
   }
   n20 = "Transform (Modify) Bone - Bone: dyn_stripB_01"(ComponentPose=@n19, Alpha=1, Translation="1,0,1", Rotation="-8,0,0", Scale="1,1,1") {
    rotationMode=BMM_Additive
    rotationSpace=BCS_BoneSpace
    translationMode=BMM_Additive
    translationSpace=BCS_BoneSpace
    alphaBoolBlend {blendOption=Linear}
    boneToModify {boneName=dyn_stripB_01}
    rotation {pitch=-8}
    translation {x=1, z=1}
   }
   n21 = "Transform (Modify) Bone - Bone: dyn_stripB_02"(ComponentPose=@n20, Alpha=1, Translation="0,0,0", Rotation="5,0,0", Scale="1,1,1") {
    rotationMode=BMM_Additive
    rotationSpace=BCS_BoneSpace
    translationMode=BMM_Additive
    translationSpace=BCS_BoneSpace
    alphaBoolBlend {blendOption=Linear}
    boneToModify {boneName=dyn_stripB_02}
    rotation {pitch=5}
   }
   Save cached pose 'RBANPose'(Pose=Component To Local(ComponentPose=@n21)) {cacheName=RBANPose}
  ]}
 EventGraph {nodes [
   Event BlueprintUpdateAnimation()
   > n1 = Sequence!(execute) -> (then_0, then_1, then_2:Exec, then_3:Exec)
   > n2 = Branch(Condition=OR Boolean!(A=Get bIsSprinting(), B=Get bIsMoving2D(), C:bool, D:bool, E:bool) -> (ReturnValue))
   > Set MovingAlpha(MovingAlpha:double=0.3) -> (Output_Get:double)
   Set MovingAlpha(execute=@n2.else, MovingAlpha=1)
   n3 = Branch(execute=@n1.then_1, Condition=OR Boolean!(A=Get bIsJumping(), B=Get bIsFalling(), C) -> (ReturnValue))
   > Set MovingAlpha(MovingAlpha=0.5)
   Set MovingAlpha(execute=@n3.else, MovingAlpha=1)
   Set is FrontEnd(execute=@n1.then_3, is FrontEnd:double="To Float (Boolean)"(InBool=Get bIsInFrontEnd())) -> (Output_Get:double)
  ]}
]
variables [
 Gravity Override {type=Vector}
 Gravity Override Params {type=GravityOverrideParamsStruct}
 is FrontEnd {type=number}
 MovingAlpha {type=number}
 Tactical jump {type=number}
]
